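; Levels-checker actor for the Gbuffer turntable that drives a post-process material to flag base color, metallic and specular values against configurable limits and colors.
; Construction script creates the dynamic material if missing, pushes thresholds, colors and max distance, then toggles its blendable weight via enable/disable.
; Setter functions update min/max values.
; Custom events toggle Time of Day Manager visibility with debug prints; EnableTOD also rotates the first BP_DaySequence actor.
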(event EventBeginPlay)

(event Collision|EventActorBeginOverlap
  (params (OtherActor :type Actor%20Object%20Reference)))

(event EventTick
  (params (DeltaSeconds :type Float%20%28single-precision%29)))

(event Custom|DisableTOD
  (Development|PrintString "----------------------Attempting Visibility Change of TODM")
  (Development|PrintString (Utilities|String|ToString(Boolean) false))
  (TimeOfDay|SetTimeOfDayManagerVisibility (TimeOfDay|GetGlobalTimeOfDayManager) false))

(event Custom|EnableTOD
  (Development|PrintString "----------------------Attempting Visibility Change of TODM")
  (Development|PrintString (Utilities|String|ToString(Boolean) true))
  (TimeOfDay|SetTimeOfDayManagerVisibility (TimeOfDay|GetGlobalTimeOfDayManager) true)
  (bind _outactors (Actor|GetAllActorsOfClass "/Game/TimeOfDay/TODM_Sequence/Blueprint/BP_DaySequence.BP_DaySequence_C"))
  (Transformation|SetActorRotation (Utilities|Array|Get(acopy) _outactors) "0.000000,160.000000,0.000000" false))

(fn disable
  (Utilities|Struct|SetmembersinPostProcessSettings (Variables|/Script/Engine.PostProcessComponent|GetSettings (Variables|Default|GetPostProcess)) (Utilities|Struct|MakeWeightedBlendables (Utilities|Array|MakeArray (Utilities|Struct|MakeWeightedBlendable 0.0 (Variables|Default|GetdynamicMaterialPP))))))

(fn enable
  (Utilities|Struct|SetmembersinPostProcessSettings (Variables|/Script/Engine.PostProcessComponent|GetSettings (Variables|Default|GetPostProcess)) (Utilities|Struct|MakeWeightedBlendables (Utilities|Array|MakeArray (Utilities|Struct|MakeWeightedBlendable 1.0 (Variables|Default|GetdynamicMaterialPP))))))

(fn setMaxValue
  (params (MaxValue :type Float%20%28double-precision%29))
  (Rendering|Material|SetScalarParameterValue (Variables|Default|GetdynamicMaterialPP) "MaxValue" MaxValue :!class /Script/Engine.MaterialInstanceDynamic)
  (Variables|Default|SetmaxValue MaxValue))

(fn setMinValue
  (params (MinValue :type Float%20%28double-precision%29))
  (Rendering|Material|SetScalarParameterValue (Variables|Default|GetdynamicMaterialPP) "MinValue" MinValue :!class /Script/Engine.MaterialInstanceDynamic)
  (Variables|Default|SetminValue MinValue))

(fn ConstructionScript
  (bind _dynamicmaterialpp (Variables|Default|GetdynamicMaterialPP))
  (if (not (Utilities|IsValid (Variables|Default|GetdynamicMaterialPP) :!class /Script/Engine.KismetSystemLibrary))
    (bind _returnvalue (Rendering|Material|CreateDynamicMaterialInstance "/Gbuffer_Turntable/FN_PublicTurntable_Assets/LevelsChecker/MAT_LevelsChecker_BaseColor.MAT_LevelsChecker_BaseColor" "None" "None" :!class /Script/Engine.KismetMaterialLibrary))
    (Variables|Default|SetdynamicMaterialPP _returnvalue))
  (Rendering|Material|SetScalarParameterValue (Variables|Default|GetdynamicMaterialPP) "MinValue" (Variables|Default|GetminValue) :!class /Script/Engine.MaterialInstanceDynamic)
  (Rendering|Material|SetScalarParameterValue (Variables|Default|GetdynamicMaterialPP) "MinMetallicValue" (* (Math|Conversions|ToInteger(Boolean) (Variables|Default|GetcheckMetallic)) (Variables|Default|GetminMetallicValue)) :!class /Script/Engine.MaterialInstanceDynamic)
  (Rendering|Material|SetScalarParameterValue (Variables|Default|GetdynamicMaterialPP) "MaxValue" (Variables|Default|GetmaxValue) :!class /Script/Engine.MaterialInstanceDynamic)
  (Rendering|Material|SetScalarParameterValue (Variables|Default|GetdynamicMaterialPP) "MaxDistance" (Variables|Default|GetmaxDistance) :!class /Script/Engine.MaterialInstanceDynamic)
  (Rendering|Material|SetVectorParameterValue (Variables|Default|GetdynamicMaterialPP) "MinValueColor" (Variables|Default|GetminValueColor) :!class /Script/Engine.MaterialInstanceDynamic)
  (Rendering|Material|SetVectorParameterValue (Variables|Default|GetdynamicMaterialPP) "MinMetallicValueColor" (Variables|Default|GetminMetallicValueColor) :!class /Script/Engine.MaterialInstanceDynamic)
  (Rendering|Material|SetVectorParameterValue (Variables|Default|GetdynamicMaterialPP) "MaxValueColor" (Variables|Default|GetmaxValueColor) :!class /Script/Engine.MaterialInstanceDynamic)
  ;[1 1 1] Set Specular Check
  (Rendering|Material|SetScalarParameterValue _dynamicmaterialpp "MinSpecular" (Variables|Default|GetMinSpecularValue) :!class /Script/Engine.MaterialInstanceDynamic)
  (Rendering|Material|SetVectorParameterValue _dynamicmaterialpp "MinSpecularValueColor" (Variables|Default|GetMinSpecularColor) :!class /Script/Engine.MaterialInstanceDynamic)
  ;[1 1 1] Enable
  (if (Variables|Default|GetEnabled)
    (CallFunction|Enable)
    (else
      (CallFunction|Disable))))
